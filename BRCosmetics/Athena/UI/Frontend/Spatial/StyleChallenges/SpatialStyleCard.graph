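# Spatial style card widget for the frontend style challenges screen, subclassing AthenaSpatialStyleCard around an inherited cosmetic item card.
# When the style image is set, the unlocked flag toggles the checkmark; the lock image starts collapsed and this graph never changes it.
# Selection plays Select-Anim, lifting the card and fading in a scaled highlight.
# Hover drives the item card's own Anim-Hover unless selected, and deselection reverses both animations.
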WidgetBlueprint /BRCosmetics/Athena/UI/Frontend/Spatial/StyleChallenges/SpatialStyleCard
animations [Select-Anim {
  binding_count=2
  length_seconds=0.15
  object=&self:Select-Anim
  sequence {bindings [
    Image_Selection {
     id=#1
     possessed_class=Image
     tracks [
      {
       type=MovieScene2DTransformTrack
       sections [{
         type=MovieScene2DTransformSection
         channels [
          Translation.X {type=MovieSceneScriptingFloatChannel}
          Translation.Y {type=MovieSceneScriptingFloatChannel}
          Angle {type=MovieSceneScriptingFloatChannel}
          Scale.X {type=MovieSceneScriptingFloatChannel, keys [[0,1],[2,1.065],[3,1.075]]}
          Scale.Y {type=MovieSceneScriptingFloatChannel, keys [[0,1],[2,1.065],[3,1.075]]}
          Shear.X {type=MovieSceneScriptingFloatChannel}
          Shear.Y {type=MovieSceneScriptingFloatChannel}
         ]
        }]
      }
      {
       type=MovieSceneFloatTrack
       display_name=Render Opacity
       sections [{type=MovieSceneFloatSection, channels [None {type=MovieSceneScriptingFloatChannel, keys [[0,0],[2,0.85],[3,1]]}]}]
      }
     ]
    }
    Overlay_0 {
     id=#2
     possessed_class=Overlay
     tracks [{
       type=MovieScene2DTransformTrack
       sections [{
         type=MovieScene2DTransformSection
         channels [
          Translation.X {type=MovieSceneScriptingFloatChannel, const [0,3]}
          Translation.Y {type=MovieSceneScriptingFloatChannel, keys [[0,0],[2,-14],[3,-16]]}
          Angle {type=MovieSceneScriptingFloatChannel}
          Scale.X {type=MovieSceneScriptingFloatChannel}
          Scale.Y {type=MovieSceneScriptingFloatChannel}
          Shear.X {type=MovieSceneScriptingFloatChannel}
          Shear.Y {type=MovieSceneScriptingFloatChannel}
         ]
        }]
      }]
    }
   ]}
 }]
graphs [EventGraph {nodes [
   n1 = Event OnStyleImageSet!() -> (OutputDelegate, then, bUnlocked:bool) {is=title K2Node_Event}
   > SetVisibility(
    self=Get Image_Active() -> (Image_Active:Image Object Reference)
    InVisibility=Select(Option 0=Collapsed, Option 1=Visible, Index=@n1.bUnlocked)
   )
   Event On Hovered()
   n2 = Get isSelected() -> (isSelected:bool)
   > n3 = Branch(Condition=@n2)
   n4 = Get ItemCard_Style() -> (ItemCard_Style:Fort Cosmetic Item Card Object Reference)
   n5 = Cast To FortCosmeticItemCard!(execute=@n3.else, Object=@n4) -> (then, CastFailed, AsFort Cosmetic Item Card:Fort Cosmetic Item Card Object Reference) {is=title K2Node_DynamicCast}
   > PlayAnimationForward(
    self=@n4
    InAnimation=Get Anim-Hover!(self:Fort Cosmetic Item Card Object Reference=@n5) -> (Anim-Hover:Widget Animation Object Reference)
    PlaybackSpeed=2
   )
   Event On Selected()
   > Set isSelected(isSelected:bool=true) -> (Output_Get:bool)
   n6 = Get Select-Anim() -> (Select-Anim:Widget Animation Object Reference)
   > PlayAnimationForward(InAnimation=@n6)
   Event On Deselected()
   > Set isSelected(isSelected=false)
   > PlayAnimationReverse(InAnimation=@n6)
   n7 = Get ItemCard_Style()
   > n8 = Cast To FortCosmeticItemCard(Object=@n7)
   > PlayAnimationReverse(self=@n7, InAnimation=Get Anim-Hover(self=@n8), PlaybackSpeed=2)
   Event On Unhovered()
   > n9 = Branch(Condition=@n2)
   n10 = Cast To FortCosmeticItemCard(execute=@n9.else, Object=@n4)
   > PlayAnimationReverse(self=@n4, InAnimation=Get Anim-Hover(self=@n10), PlaybackSpeed=2)
  ]}]
variables [isSelected {type=boolean}]
widgets {
 info {inheritedWidgetCount=2, parentClass=/Script/FortniteUI.AthenaSpatialStyleCard, rootWidgetClass=&Overlay, widgetCount=7}
 widgets [
  Overlay_23 {widgetClassPath=&Overlay, widget=Overlay_23}
  Overlay_0 {widgetClassPath=&Overlay, parent=Overlay_23, widget=Overlay_0, slot {class=OverlaySlot}}
  Image_Selection {
   widgetClassPath=&Image
   bIsVariable=true
   parent=Overlay_0
   widget=Image_Selection
   properties {renderOpacity=0, renderTransform {angle=0, scale=<1.075 1.075>, shear=<0 0>, translation=<0 0>}}
   slot {class=OverlaySlot, properties {horizontalAlignment=HAlign_Fill, verticalAlignment=VAlign_Fill}}
  }
  ItemCard_Style {
   widgetClassPath=/FortniteCoreUI/Athena/UI/Frontend/CosmeticItemCard/FortCosmeticItemCard
   bInherited=true
   bIsVariable=true
   parent=Overlay_0
   widget=ItemCard_Style
   slot {class=OverlaySlot}
  }
  Image_Locked {
   widgetClassPath=&Image
   bIsVariable=true
   parent=Overlay_0
   widget=Image_Locked
   properties {
    visibility=Collapsed
    brush {
     bIsDynamicallyLoaded=false
     drawAs=Image
     imageType=NoImage
     mirroring=NoMirror
     resourceName=None
     resourceObject=/FortUILibrary/Styles/Border/Materials/M_CircledLock
     tiling=NoTile
     imageSize=<32 32>
     margin=<ltrb 0 0 0 0>
     outlineSettings {
      bUseBrushTransparency=false
      roundingType=HalfHeightRadius
      width=0
      color {colorUseRule=UseColor_Specified, specifiedColor=<rgba 0 0 0 0>}
      cornerRadii=<0 0 0 0>
     }
     tintColor {colorUseRule=UseColor_Specified, specifiedColor=<rgba 1 1 1 1>}
     uVRegion {bIsValid=false, max=<0 0>, min=<0 0>}
    }
    renderTransform {angle=0, scale=<1.5 1.5>, shear=<0 0>, translation=<0 0>}
   }
   slot {class=OverlaySlot, properties {horizontalAlignment=HAlign_Center, verticalAlignment=VAlign_Center}}
  }
  Image_Active {
   widgetClassPath=&Image
   bIsVariable=true
   parent=Overlay_0
   widget=Image_Active
   properties {
    visibility=SelfHitTestInvisible
    brush {
     bIsDynamicallyLoaded=false
     drawAs=Image
     imageType=NoImage
     mirroring=NoMirror
     resourceName=None
     resourceObject=/Game/Athena/UI/Challenges/Art/T_UI_ChallengeCheck_64
     tiling=NoTile
     imageSize=<32 32>
     margin=<ltrb 0 0 0 0>
     outlineSettings {
      bUseBrushTransparency=false
      roundingType=HalfHeightRadius
      width=0
      color {colorUseRule=UseColor_Specified, specifiedColor=<rgba 0 0 0 0>}
      cornerRadii=<0 0 0 0>
     }
     tintColor {colorUseRule=UseColor_Specified, specifiedColor=<rgba 1 1 1 1>}
     uVRegion {bIsValid=false, max=<0 0>, min=<0 0>}
    }
    renderTransform {angle=0, scale=<1 1>, shear=<0 0>, translation=<6 -8>}
   }
   slot {class=OverlaySlot, properties {horizontalAlignment=HAlign_Right}}
  }
  InputActionWidget {
   widgetClassPath=&CommonActionWidget
   bInherited=true
   bIsVariable=true
   parent=Overlay_0
   widget=InputActionWidget
   properties {
    visibility=HitTestInvisible
    iconRimBrush {
     bIsDynamicallyLoaded=false
     drawAs=Image
     imageType=NoImage
     mirroring=NoMirror
     resourceName=None
     resourceObject=/FortUILibrary/Foundation/Buttons/Materials/M-Button-Console-YellowOutline
     tiling=NoTile
     imageSize=<40 40>
     margin=<ltrb 0 0 0 0>
     outlineSettings {
      bUseBrushTransparency=false
      roundingType=HalfHeightRadius
      width=0
      color {colorUseRule=UseColor_Specified, specifiedColor=<rgba 0 0 0 0>}
      cornerRadii=<0 0 0 0>
     }
     tintColor {colorUseRule=UseColor_Specified, specifiedColor=<rgba 1 1 1 1>}
     uVRegion {bIsValid=false, max=<0 0>, min=<0 0>}
    }
    renderTransform {angle=0, scale=<1 1>, shear=<0 0>, translation=<0 20>}
   }
   slot {class=OverlaySlot, properties {horizontalAlignment=HAlign_Center, verticalAlignment=VAlign_Bottom}}
  }
 ]
}
