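# Widget Blueprint row showing one unlock reason for a style challenge in the frontend spatial UI, with check box, description and mission progress.
# Events from the C++ parent AthenaSpatialStyleChallengeUnlockReason format Battle Pass level or purchase text and the completed/required count.
# OnUnlockReasonUpdated switches the check icon, toggles "Done!" and recolors the description, applying Locked-Color when unlocked and Unlocked-Color otherwise, opposite to their names.
# The graph leaves the inherited progress bars untouched.
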WidgetBlueprint /BRCosmetics/Athena/UI/Frontend/Spatial/StyleChallenges/SpatialStyleChallengeUnlockReason
graphs [EventGraph {nodes [
   n1 = Event ShowBattlepassLevelReason!() -> (OutputDelegate, then, InLevel:int=0) {is=title K2Node_Event}
   > SetText(
    self=Get RichText_Description() -> (RichText_Description:Common Rich Text Block Object Reference)
    InText=Format Text!(Format="LOCTEXT(\"Reach Battle Pass Level {level}\")", level:int=@n1.InLevel) -> (Result)
   )
   n2 = Event ShowBattlepassPurchaseReason!() -> (OutputDelegate, then, InSeason:"Text (by ref)") {is=title K2Node_Event}
   > SetText(
    self=Get RichText_Description()
    InText=Format Text!(Format="LOCTEXT(\"Buy {season} Battle Pass\")", season:"Text (by ref)"=@n2.InSeason) -> (Result)
   )
   n3 = Event UpdateMissionProgressCount!() -> (OutputDelegate, then, CompletedCount:int=0, RequiredCount:int=0) {is=title K2Node_Event}
   > SetText(
    self=Get RichText_MissionProgressCount() -> (RichText_MissionProgressCount:Common Rich Text Block Object Reference)
    InText=Format Text!(
     Format="LOCTEXT(\"<text color=\\\"FFF\\\">{CompletedCount}</> / {RequiredCount}\")"
     CompletedCount:int=@n3.CompletedCount
     RequiredCount:int=@n3.RequiredCount
    ) -> (Result)
   )
   n4 = Event OnUnlockReasonUpdated!() -> (OutputDelegate, then, bInUnlocked:bool) {is=title K2Node_Event}
   > SetActiveWidgetIndex(
    self=Get WidgetSwitcher_CompleteCheck() -> (WidgetSwitcher_CompleteCheck:Widget Switcher Object Reference)
    Index=SelectInt(A=0, B=1, bPickA=@n4.bInUnlocked)
   )
   > SetVisibility(
    self=Get DoneTag() -> (DoneTag:Common Text Object Reference)
    InVisibility=Select(Option 0=Collapsed, Option 1=SelfHitTestInvisible, Index=@n4.bInUnlocked)
   )
   > SetDefaultColorAndOpacity!(
    execute
    self:Rich Text Block Object Reference=Get RichText_Description()
    InColorAndOpacity:Slate Color Structure=Select(
     Option 0:Slate Color Structure=Get Unlocked-Color() -> (Unlocked-Color:Slate Color Structure)
     Option 1:Slate Color Structure=Get Locked-Color() -> (Locked-Color:Slate Color Structure)
     Index=@n4.bInUnlocked
    ) -> (ReturnValue:Slate Color Structure)
   ) -> (then)
  ]}]
variables [
 Locked-Color {type=SlateColor, default {colorUseRule=UseColor_Specified, specifiedColor=<rgba 0 0.2502 1 1>}}
 Unlocked-Color {type=SlateColor, default {colorUseRule=UseColor_Specified, specifiedColor=<rgba 0.5583 0.9473 1 1>}}
]
widgets {
 info {
  inheritedWidgetCount=4
  parentClass=/Script/FortniteUI.AthenaSpatialStyleChallengeUnlockReason
  rootWidgetClass=&SizeBox
  widgetCount=15
 }
 widgets [
  SizeBox_1 {widgetClassPath=&SizeBox, widget=SizeBox_1, properties {bOverride_WidthOverride=true, widthOverride=512}}
  VerticalBox_2 {widgetClassPath=&VerticalBox, parent=SizeBox_1, widget=VerticalBox_2, slot {class=SizeBoxSlot}}
  HorizontalBox_31 {widgetClassPath=&HorizontalBox, parent=VerticalBox_2, widget=HorizontalBox_31, slot {class=VerticalBoxSlot}}
  SizeBox_0 {
   widgetClassPath=&SizeBox
   parent=HorizontalBox_31
   widget=SizeBox_0
   properties {bOverride_HeightOverride=true, bOverride_WidthOverride=true, heightOverride=40, widthOverride=40}
   slot {class=HorizontalBoxSlot, properties {horizontalAlignment=HAlign_Left, verticalAlignment=VAlign_Center, padding=<ltrb 0 0 24 0>}}
  }
  WidgetSwitcher_CompleteCheck {
   widgetClassPath=&WidgetSwitcher
   bIsVariable=true
   parent=SizeBox_0
   widget=WidgetSwitcher_CompleteCheck
   slot {class=SizeBoxSlot}
  }
  Complete {
   widgetClassPath=&FortLazyImage
   bIsVariable=true
   parent=WidgetSwitcher_CompleteCheck
   widget=Complete
   properties {brush {
     bIsDynamicallyLoaded=false
     drawAs=Image
     imageType=NoImage
     mirroring=NoMirror
     resourceName=None
     resourceObject=/Game/Athena/UI/Challenges/Art/T_UI_ChallengeCheck_64
     tiling=NoTile
     imageSize=<32 32>
     margin=<ltrb 0 0 0 0>
     outlineSettings {
      bUseBrushTransparency=false
      roundingType=HalfHeightRadius
      width=0
      color {colorUseRule=UseColor_Specified, specifiedColor=<rgba 0 0 0 0>}
      cornerRadii=<0 0 0 0>
     }
     tintColor {colorUseRule=UseColor_Specified, specifiedColor=<rgba 1 1 1 1>}
     uVRegion {bIsValid=false, max=<0 0>, min=<0 0>}
    }}
   slot {class=WidgetSwitcherSlot}
  }
  Empty {
   widgetClassPath=&FortLazyImage
   bIsVariable=true
   parent=WidgetSwitcher_CompleteCheck
   widget=Empty
   properties {
    brush {
     bIsDynamicallyLoaded=false
     drawAs=Image
     imageType=NoImage
     mirroring=NoMirror
     resourceName=None
     resourceObject=/FortUILibrary/Foundation/Art_ToolKit/AngledBoxes/M_UI_AngledCheckBox_Empty
     tiling=NoTile
     imageSize=<32 32>
     margin=<ltrb 0 0 0 0>
     outlineSettings {
      bUseBrushTransparency=false
      roundingType=HalfHeightRadius
      width=0
      color {colorUseRule=UseColor_Specified, specifiedColor=<rgba 0 0 0 0>}
      cornerRadii=<0 0 0 0>
     }
     tintColor {colorUseRule=UseColor_Specified, specifiedColor=<rgba 1 1 1 1>}
     uVRegion {bIsValid=false, max=<0 0>, min=<0 0>}
    }
    colorAndOpacity=<rgba 0.5583 0.9473 1 1>
   }
   slot {class=WidgetSwitcherSlot}
  }
  DoneTag {
   widgetClassPath=&CommonTextBlock
   bIsVariable=true
   parent=HorizontalBox_31
   widget=DoneTag
   properties {
    text="Done!"
    textTransformPolicy=ToUpper
    colorAndOpacity {colorUseRule=UseColor_Specified, specifiedColor=<rgba 0.0497 1 0.2384 1>}
    font {
     bForceMonospaced=false
     bMaterialIsStencil=false
     fontMaterial=null
     fontObject=&BurbankSmall
     letterSpacing=0
     monospacedWidth=1
     size=20
     skewAmount=0
     typefaceFontName=Black
     outlineSettings {
      bApplyOutlineToDropShadows=false
      bMiteredCorners=false
      bSeparateFillAlpha=false
      outlineMaterial=null
      outlineSize=0
      outlineColor=<rgba 0 0 0 1>
     }
    }
    renderTransform {angle=0, scale=<1 1>, shear=<-12 0>, translation=<0 0>}
    shadowOffset=<0 0>
   }
   slot {class=HorizontalBoxSlot, properties {verticalAlignment=VAlign_Center, padding=<ltrb 0 0 12 0>}}
  }
  RichText_Description {
   widgetClassPath=&CommonRichTextBlock
   bInherited=true
   bIsVariable=true
   parent=HorizontalBox_31
   widget=RichText_Description
   properties {
    bOverrideDefaultStyle=true
    scrollStyle=&TextScrollStyle-Base_C
    text="{Long description of objective.}"
    defaultTextStyleOverride {
     overflowPolicy=Clip
     transformPolicy=None
     colorAndOpacity {colorUseRule=UseColor_Specified, specifiedColor=<rgba 0 0.2502 1 1>}
     font {
      bForceMonospaced=false
      bMaterialIsStencil=false
      fontMaterial=null
      fontObject=&BurbankSmall
      letterSpacing=0
      monospacedWidth=1
      size=18
      skewAmount=0
      typefaceFontName=Bold
      outlineSettings {
       bApplyOutlineToDropShadows=false
       bMiteredCorners=false
       bSeparateFillAlpha=false
       outlineMaterial=null
       outlineSize=0
       outlineColor=<rgba 0 0 0 1>
      }
     }
     highlightColor {colorUseRule=UseColor_Specified, specifiedColor=<rgba 0 0 0 1>}
     highlightShape {
      bIsDynamicallyLoaded=false
      drawAs=Image
      imageType=NoImage
      mirroring=NoMirror
      resourceName=None
      resourceObject=null
      tiling=NoTile
      imageSize=<32 32>
      margin=<ltrb 0 0 0 0>
      outlineSettings {
       bUseBrushTransparency=false
       roundingType=HalfHeightRadius
       width=0
       color {colorUseRule=UseColor_Specified, specifiedColor=<rgba 0 0 0 0>}
       cornerRadii=<0 0 0 0>
      }
      tintColor {colorUseRule=UseColor_Specified, specifiedColor=<rgba 1 1 1 1>}
      uVRegion {bIsValid=false, max=<0 0>, min=<0 0>}
     }
     selectedBackgroundColor {colorUseRule=UseColor_Foreground, specifiedColor=<rgba 1 0 1 1>}
     shadowColorAndOpacity=<rgba 0 0 0 1>
     shadowOffset=<0 0>
     strikeBrush {
      bIsDynamicallyLoaded=false
      drawAs=Image
      imageType=NoImage
      mirroring=NoMirror
      resourceName=None
      resourceObject=null
      tiling=NoTile
      imageSize=<32 32>
      margin=<ltrb 0 0 0 0>
      outlineSettings {
       bUseBrushTransparency=false
       roundingType=HalfHeightRadius
       width=0
       color {colorUseRule=UseColor_Specified, specifiedColor=<rgba 0 0 0 0>}
       cornerRadii=<0 0 0 0>
      }
      tintColor {colorUseRule=UseColor_Specified, specifiedColor=<rgba 1 1 1 1>}
      uVRegion {bIsValid=false, max=<0 0>, min=<0 0>}
     }
     underlineBrush {
      bIsDynamicallyLoaded=false
      drawAs=Image
      imageType=NoImage
      mirroring=NoMirror
      resourceName=None
      resourceObject=null
      tiling=NoTile
      imageSize=<32 32>
      margin=<ltrb 0 0 0 0>
      outlineSettings {
       bUseBrushTransparency=false
       roundingType=HalfHeightRadius
       width=0
       color {colorUseRule=UseColor_Specified, specifiedColor=<rgba 0 0 0 0>}
       cornerRadii=<0 0 0 0>
      }
      tintColor {colorUseRule=UseColor_Specified, specifiedColor=<rgba 1 1 1 1>}
      uVRegion {bIsValid=false, max=<0 0>, min=<0 0>}
     }
    }
   }
   slot {class=HorizontalBoxSlot, properties {verticalAlignment=VAlign_Center, size {sizeRule=Fill, value=1}}}
  }
  HBox_MissionProgress {
   widgetClassPath=&HorizontalBox
   bInherited=true
   parent=VerticalBox_2
   widget=HBox_MissionProgress
   slot {class=VerticalBoxSlot, properties {padding=<ltrb 0 8 0 0>}}
  }
  SizeBox_2 {
   widgetClassPath=&SizeBox
   parent=HBox_MissionProgress
   widget=SizeBox_2
   properties {bOverride_HeightOverride=true, bOverride_WidthOverride=true, heightOverride=12, widthOverride=150}
   slot {class=HorizontalBoxSlot, properties {verticalAlignment=VAlign_Center, padding=<ltrb 64 0 0 0>, size {sizeRule=Fill, value=1}}}
  }
  Overlay_3 {widgetClassPath=&Overlay, parent=SizeBox_2, widget=Overlay_3, slot {class=SizeBoxSlot}}
  ProgressBar_PercentComplete_Front {
   widgetClassPath=/Script/UMG.ProgressBar
   bInherited=true
   bIsVariable=true
   parent=Overlay_3
   widget=ProgressBar_PercentComplete_Front
   properties {
    percent=0.25
    fillColorAndOpacity=<rgba 0.0331 0.855 1 1>
    widgetStyle {
     enableFillAnimation=false
     backgroundImage {
      bIsDynamicallyLoaded=false
      drawAs=Box
      imageType=FullColor
      mirroring=NoMirror
      resourceName=../../../Engine/Content/Slate/Common/ProgressBar_Background.png
      resourceObject=/FortniteCoreUI/Frontend/Spatial/Art/M_Ui_Spacial_Seaparator_Background_progress
      tiling=NoTile
      imageSize=<0 0>
      margin=<ltrb 0.2 0.2 0.2 0.2>
      outlineSettings {
       bUseBrushTransparency=false
       roundingType=HalfHeightRadius
       width=0
       color {colorUseRule=UseColor_Specified, specifiedColor=<rgba 0 0 0 0>}
       cornerRadii=<0 0 0 0>
      }
      tintColor {colorUseRule=UseColor_Specified, specifiedColor=<rgba 0 0.2502 1 0.25>}
      uVRegion {bIsValid=false, max=<0 0>, min=<0 0>}
     }
     fillImage {
      bIsDynamicallyLoaded=false
      drawAs=Box
      imageType=FullColor
      mirroring=NoMirror
      resourceName=../../../Engine/Content/Slate/Common/ProgressBar_Fill.png
      resourceObject=&M_Ui_Spacial_Seaparator_Background_progress
      tiling=NoTile
      imageSize=<0 0>
      margin=<ltrb 0.25 0.25 0.25 0.25>
      outlineSettings {
       bUseBrushTransparency=false
       roundingType=HalfHeightRadius
       width=0
       color {colorUseRule=UseColor_Specified, specifiedColor=<rgba 0 0 0 0>}
       cornerRadii=<0 0 0 0>
      }
      tintColor {colorUseRule=UseColor_Specified, specifiedColor=<rgba 0.5583 0.9473 1 1>}
      uVRegion {bIsValid=false, max=<0 0>, min=<0 0>}
     }
     marqueeImage {
      bIsDynamicallyLoaded=false
      drawAs=Image
      imageType=FullColor
      mirroring=NoMirror
      resourceName=../../../Engine/Content/Slate/Common/ProgressBar_Marquee.png
      resourceObject=null
      tiling=Horizontal
      imageSize=<20 12>
      margin=<ltrb 0 0 0 0>
      outlineSettings {
       bUseBrushTransparency=false
       roundingType=HalfHeightRadius
       width=0
       color {colorUseRule=UseColor_Specified, specifiedColor=<rgba 0 0 0 0>}
       cornerRadii=<0 0 0 0>
      }
      tintColor {colorUseRule=UseColor_Specified, specifiedColor=<rgba 1 1 1 1>}
      uVRegion {bIsValid=false, max=<0 0>, min=<0 0>}
     }
    }
   }
   slot {class=OverlaySlot, properties {horizontalAlignment=HAlign_Fill, verticalAlignment=VAlign_Fill}}
  }
  ProgressBar_PercentComplete_Back {
   widgetClassPath=&ProgressBar
   bInherited=true
   bIsVariable=true
   parent=Overlay_3
   widget=ProgressBar_PercentComplete_Back
   properties {
    percent=0.5524
    visibility=Collapsed
    widgetStyle {
     enableFillAnimation=false
     backgroundImage {
      bIsDynamicallyLoaded=false
      drawAs=Box
      imageType=FullColor
      mirroring=NoMirror
      resourceName=../../../Engine/Content/Slate/Common/ProgressBar_Background.png
      resourceObject=&M_Ui_Spacial_Seaparator_Background_progress
      tiling=NoTile
      imageSize=<8 8>
      margin=<ltrb 0.25 0.25 0.25 0.25>
      outlineSettings {
       bUseBrushTransparency=false
       roundingType=HalfHeightRadius
       width=0
       color {colorUseRule=UseColor_Specified, specifiedColor=<rgba 0 0 0 0>}
       cornerRadii=<0 0 0 0>
      }
      tintColor {colorUseRule=UseColor_Specified, specifiedColor=<rgba 0.0331 0.855 1 0.098>}
      uVRegion {bIsValid=false, max=<0 0>, min=<0 0>}
     }
     fillImage {
      bIsDynamicallyLoaded=false
      drawAs=Box
      imageType=FullColor
      mirroring=NoMirror
      resourceName=../../../Engine/Content/Slate/Common/ProgressBar_Fill.png
      resourceObject=&M_Ui_Spacial_Seaparator_Background_progress
      tiling=NoTile
      imageSize=<8 8>
      margin=<ltrb 0.25 0.25 0.25 0.25>
      outlineSettings {
       bUseBrushTransparency=false
       roundingType=HalfHeightRadius
       width=0
       color {colorUseRule=UseColor_Specified, specifiedColor=<rgba 0 0 0 0>}
       cornerRadii=<0 0 0 0>
      }
      tintColor {colorUseRule=UseColor_Specified, specifiedColor=<rgba 0.0331 0.855 1 0.075>}
      uVRegion {bIsValid=false, max=<0 0>, min=<0 0>}
     }
     marqueeImage {
      bIsDynamicallyLoaded=false
      drawAs=Image
      imageType=FullColor
      mirroring=NoMirror
      resourceName=../../../Engine/Content/Slate/Common/ProgressBar_Marquee.png
      resourceObject=null
      tiling=Horizontal
      imageSize=<20 12>
      margin=<ltrb 0 0 0 0>
      outlineSettings {
       bUseBrushTransparency=false
       roundingType=HalfHeightRadius
       width=0
       color {colorUseRule=UseColor_Specified, specifiedColor=<rgba 0 0 0 0>}
       cornerRadii=<0 0 0 0>
      }
      tintColor {colorUseRule=UseColor_Specified, specifiedColor=<rgba 1 1 1 1>}
      uVRegion {bIsValid=false, max=<0 0>, min=<0 0>}
     }
    }
   }
   slot {class=OverlaySlot, properties {horizontalAlignment=HAlign_Fill, verticalAlignment=VAlign_Fill}}
  }
  RichText_MissionProgressCount {
   widgetClassPath=&CommonRichTextBlock
   bIsVariable=true
   parent=HBox_MissionProgress
   widget=RichText_MissionProgressCount
   properties {
    bOverrideDefaultStyle=true
    text="1<text color=\"0089FF\"> / 5</>"
    defaultTextStyleOverride {
     overflowPolicy=Clip
     transformPolicy=None
     colorAndOpacity {colorUseRule=UseColor_Specified, specifiedColor=<rgba 0.5583 0.9473 1 1>}
     font {
      bForceMonospaced=false
      bMaterialIsStencil=false
      fontMaterial=null
      fontObject=&BurbankSmall
      letterSpacing=0
      monospacedWidth=1
      size=18
      skewAmount=0
      typefaceFontName=Black
      outlineSettings {
       bApplyOutlineToDropShadows=false
       bMiteredCorners=false
       bSeparateFillAlpha=false
       outlineMaterial=null
       outlineSize=0
       outlineColor=<rgba 0 0 0 1>
      }
     }
     highlightColor {colorUseRule=UseColor_Specified, specifiedColor=<rgba 0 0 0 1>}
     highlightShape {
      bIsDynamicallyLoaded=false
      drawAs=Image
      imageType=NoImage
      mirroring=NoMirror
      resourceName=None
      resourceObject=null
      tiling=NoTile
      imageSize=<32 32>
      margin=<ltrb 0 0 0 0>
      outlineSettings {
       bUseBrushTransparency=false
       roundingType=HalfHeightRadius
       width=0
       color {colorUseRule=UseColor_Specified, specifiedColor=<rgba 0 0 0 0>}
       cornerRadii=<0 0 0 0>
      }
      tintColor {colorUseRule=UseColor_Specified, specifiedColor=<rgba 1 1 1 1>}
      uVRegion {bIsValid=false, max=<0 0>, min=<0 0>}
     }
     selectedBackgroundColor {colorUseRule=UseColor_Foreground, specifiedColor=<rgba 1 0 1 1>}
     shadowColorAndOpacity=<rgba 0 0 0 1>
     shadowOffset=<0 0>
     strikeBrush {
      bIsDynamicallyLoaded=false
      drawAs=Image
      imageType=NoImage
      mirroring=NoMirror
      resourceName=None
      resourceObject=null
      tiling=NoTile
      imageSize=<32 32>
      margin=<ltrb 0 0 0 0>
      outlineSettings {
       bUseBrushTransparency=false
       roundingType=HalfHeightRadius
       width=0
       color {colorUseRule=UseColor_Specified, specifiedColor=<rgba 0 0 0 0>}
       cornerRadii=<0 0 0 0>
      }
      tintColor {colorUseRule=UseColor_Specified, specifiedColor=<rgba 1 1 1 1>}
      uVRegion {bIsValid=false, max=<0 0>, min=<0 0>}
     }
     underlineBrush {
      bIsDynamicallyLoaded=false
      drawAs=Image
      imageType=NoImage
      mirroring=NoMirror
      resourceName=None
      resourceObject=null
      tiling=NoTile
      imageSize=<32 32>
      margin=<ltrb 0 0 0 0>
      outlineSettings {
       bUseBrushTransparency=false
       roundingType=HalfHeightRadius
       width=0
       color {colorUseRule=UseColor_Specified, specifiedColor=<rgba 0 0 0 0>}
       cornerRadii=<0 0 0 0>
      }
      tintColor {colorUseRule=UseColor_Specified, specifiedColor=<rgba 1 1 1 1>}
      uVRegion {bIsValid=false, max=<0 0>, min=<0 0>}
     }
    }
    renderTransform {angle=0, scale=<1 1>, shear=<-12 0>, translation=<0 0>}
   }
   slot {class=HorizontalBoxSlot, properties {verticalAlignment=VAlign_Center, padding=<ltrb 28 0 0 0>}}
  }
 ]
}
