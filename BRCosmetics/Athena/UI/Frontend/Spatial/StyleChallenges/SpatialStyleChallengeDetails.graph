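# Right-side frontend panel detailing a cosmetic's style challenges: style cards, prerequisites, main mission progress, a completion view and a Battle Pass purchase button.
# Extends the native AthenaSpatialStyleChallengeDetails, which supplies update events and the inherited entry boxes.
# HandlePrerequisitesAndMissionUpdated builds the title from character and parenthesized style name, appending UNLOCKED! when not locked.
# Character and style updates play staggered PageRefresh and StyleRefresh animations; PlayIntroAnim plays the slide-in Intro.
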WidgetBlueprint /BRCosmetics/Athena/UI/Frontend/Spatial/StyleChallenges/SpatialStyleChallengeDetails
animations [
 Intro {
  binding_count=9
  length_seconds=0.45
  object=&self:Intro
  sequence {bindings [
    EntryBox_Styles {
     id=#1
     possessed_class=DynamicEntryBox
     tracks [
      {
       type=MovieScene2DTransformTrack
       sections [{type=MovieScene2DTransformSection, channels [Translation.X {type=MovieSceneScriptingFloatChannel, keys [[1,2000],[6,0]]}]}]
      }
      {
       type=MovieSceneFloatTrack
       display_name=Render Opacity
       sections [{type=MovieSceneFloatSection, channels [None {type=MovieSceneScriptingFloatChannel, const [1,1]}]}]
      }
     ]
    }
    RichText_UnlockStyle {
     id=#2
     possessed_class=CommonRichTextBlock
     tracks [{
       type=MovieScene2DTransformTrack
       sections [{type=MovieScene2DTransformSection, channels [Translation.X {type=MovieSceneScriptingFloatChannel, keys [[1,2000],[6,0]]}]}]
      }]
    }
    EntryBox_Prerequisites {
     id=#3
     possessed_class=DynamicEntryBox
     tracks [{
       type=MovieScene2DTransformTrack
       sections [{type=MovieScene2DTransformSection, channels [Translation.X {type=MovieSceneScriptingFloatChannel, keys [[4,1000],[8,0]]}]}]
      }]
    }
    RichText_CompleteText {
     id=#4
     possessed_class=CommonRichTextBlock
     tracks [{
       type=MovieScene2DTransformTrack
       sections [{type=MovieScene2DTransformSection, channels [Translation.X {type=MovieSceneScriptingFloatChannel, keys [[0,2000],[5,0]]}]}]
      }]
    }
    CommonTextBlock_175 {
     id=#5
     possessed_class=CommonTextBlock
     tracks [{
       type=MovieScene2DTransformTrack
       sections [{type=MovieScene2DTransformSection, channels [Translation.X {type=MovieSceneScriptingFloatChannel, keys [[5,1000],[9,0]]}]}]
      }]
    }
    Backplate {
     id=#6
     possessed_class=Image
     tracks [
      {
       type=MovieSceneFloatTrack
       display_name=Render Opacity
       sections [{type=MovieSceneFloatSection, channels [None {type=MovieSceneScriptingFloatChannel, keys [[0,0],[6,1]]}]}]
      }
      {
       type=MovieScene2DTransformTrack
       sections [{type=MovieScene2DTransformSection, channels [Translation.X {type=MovieSceneScriptingFloatChannel, keys [[0,2000],[5,0]]}]}]
      }
     ]
    }
    Text_Prerequisites {
     id=#7
     possessed_class=CommonTextBlock
     tracks [{
       type=MovieScene2DTransformTrack
       sections [{type=MovieScene2DTransformSection, channels [Translation.X {type=MovieSceneScriptingFloatChannel, keys [[4,1000],[8,0]]}]}]
      }]
    }
    Button_PurchaseBattlepass {
     id=#8
     possessed_class=Spatial_Specialbutton_C
     tracks [{
       type=MovieScene2DTransformTrack
       sections [{type=MovieScene2DTransformSection, channels [Translation.X {type=MovieSceneScriptingFloatChannel, keys [[5,1000],[9,0]]}]}]
      }]
    }
    QuestProgression_MainMission {
     id=#9
     possessed_class=SpatialStyleChallengeUnlockReason_C
     tracks [{
       type=MovieScene2DTransformTrack
       sections [{type=MovieScene2DTransformSection, channels [Translation.X {type=MovieSceneScriptingFloatChannel, keys [[5,1000],[9,0]]}]}]
      }]
    }
   ]}
 }
 PageRefresh {
  binding_count=8
  length_seconds=0.5
  object=&self:PageRefresh
  sequence {bindings [
    EntryBox_Styles {
     id=#1
     possessed_class=DynamicEntryBox
     tracks [
      {
       type=MovieScene2DTransformTrack
       sections [{
         type=MovieScene2DTransformSection
         channels [
          Translation.X {type=MovieSceneScriptingFloatChannel, keys [[0,2000],[3,0]]}
          Translation.Y {type=MovieSceneScriptingFloatChannel, const [0,2]}
         ]
        }]
      }
      {
       type=MovieSceneFloatTrack
       display_name=Render Opacity
       sections [{type=MovieSceneFloatSection, channels [None {type=MovieSceneScriptingFloatChannel, keys [[1,0],[3,1]]}]}]
      }
      {type=MovieSceneFloatVectorTrack, display_name=Entry Spacing, sections []}
     ]
    }
    RichText_UnlockStyle {
     id=#2
     possessed_class=CommonRichTextBlock
     tracks [
      {
       type=MovieScene2DTransformTrack
       sections [{
         type=MovieScene2DTransformSection
         channels [
          Translation.X {type=MovieSceneScriptingFloatChannel, keys [[1,2000],[4,0]]}
          Translation.Y {type=MovieSceneScriptingFloatChannel, const [0,2]}
         ]
        }]
      }
      {
       type=MovieSceneFloatTrack
       display_name=Render Opacity
       sections [{type=MovieSceneFloatSection, channels [None {type=MovieSceneScriptingFloatChannel, keys [[1,0],[4,1]]}]}]
      }
     ]
    }
    EntryBox_Prerequisites {
     id=#3
     possessed_class=DynamicEntryBox
     tracks [
      {
       type=MovieScene2DTransformTrack
       sections [{type=MovieScene2DTransformSection, channels [Translation.Y {type=MovieSceneScriptingFloatChannel, keys [[6,-24],[8,0]]}]}]
      }
      {
       type=MovieSceneFloatTrack
       display_name=Render Opacity
       sections [{type=MovieSceneFloatSection, channels [None {type=MovieSceneScriptingFloatChannel, keys [[6,0],[8,1]]}]}]
      }
     ]
    }
    RichText_CompleteText {
     id=#4
     possessed_class=CommonRichTextBlock
     tracks [{
       type=MovieScene2DTransformTrack
       sections [{type=MovieScene2DTransformSection, channels [Translation.X {type=MovieSceneScriptingFloatChannel, const [0,1]}]}]
      }]
    }
    CommonTextBlock_175 {
     id=#5
     possessed_class=CommonTextBlock
     tracks [
      {
       type=MovieScene2DTransformTrack
       sections [{type=MovieScene2DTransformSection, channels [Translation.Y {type=MovieSceneScriptingFloatChannel, const [0,1]}]}]
      }
      {
       type=MovieSceneFloatTrack
       display_name=Render Opacity
       sections [{type=MovieSceneFloatSection, channels [None {type=MovieSceneScriptingFloatChannel, keys [[7,0],[9,1]]}]}]
      }
     ]
    }
    Backplate {
     id=#6
     possessed_class=Image
     tracks [
      {
       type=MovieSceneFloatTrack
       display_name=Render Opacity
       sections [{type=MovieSceneFloatSection, channels [None {type=MovieSceneScriptingFloatChannel, const [1,1]}]}]
      }
      {
       type=MovieScene2DTransformTrack
       sections [{type=MovieScene2DTransformSection, channels [Translation.X {type=MovieSceneScriptingFloatChannel, const [0,1]}]}]
      }
     ]
    }
    Text_Prerequisites {
     id=#7
     possessed_class=CommonTextBlock
     tracks [
      {
       type=MovieScene2DTransformTrack
       sections [{
         type=MovieScene2DTransformSection
         channels [
          Translation.X {type=MovieSceneScriptingFloatChannel}
          Translation.Y {type=MovieSceneScriptingFloatChannel, const [0,1]}
         ]
        }]
      }
      {
       type=MovieSceneFloatTrack
       display_name=Render Opacity
       sections [{type=MovieSceneFloatSection, channels [None {type=MovieSceneScriptingFloatChannel, keys [[5,0],[7,1]]}]}]
      }
     ]
    }
    QuestProgression_MainMission {
     id=#9
     possessed_class=SpatialStyleChallengeUnlockReason_C
     tracks [
      {
       type=MovieScene2DTransformTrack
       sections [{type=MovieScene2DTransformSection, channels [Translation.Y {type=MovieSceneScriptingFloatChannel, keys [[8,-24],[10,0]]}]}]
      }
      {
       type=MovieSceneFloatTrack
       display_name=Render Opacity
       sections [{type=MovieSceneFloatSection, channels [None {type=MovieSceneScriptingFloatChannel, keys [[8,0],[10,1]]}]}]
      }
     ]
    }
   ]}
 }
 StyleRefresh {
  binding_count=5
  length_seconds=0.25
  object=&self:StyleRefresh
  sequence {bindings [
    EntryBox_Prerequisites {
     id=#3
     possessed_class=DynamicEntryBox
     tracks [
      {
       type=MovieScene2DTransformTrack
       sections [{
         type=MovieScene2DTransformSection
         channels [
          Translation.X {type=MovieSceneScriptingFloatChannel}
          Translation.Y {type=MovieSceneScriptingFloatChannel, keys [[1,-24],[3,0]]}
         ]
        }]
      }
      {
       type=MovieSceneFloatTrack
       display_name=Render Opacity
       sections [{type=MovieSceneFloatSection, channels [None {type=MovieSceneScriptingFloatChannel, keys [[1,0],[3,1]]}]}]
      }
     ]
    }
    RichText_UnlockStyle {
     id=#10
     possessed_class=CommonRichTextBlock
     tracks [
      {
       type=MovieScene2DTransformTrack
       sections [{type=MovieScene2DTransformSection, channels [Translation.Y {type=MovieSceneScriptingFloatChannel, keys [[0,12],[2,0]]}]}]
      }
      {
       type=MovieSceneFloatTrack
       display_name=Render Opacity
       sections [{type=MovieSceneFloatSection, channels [None {type=MovieSceneScriptingFloatChannel, keys [[0,0],[2,1]]}]}]
      }
     ]
    }
    CommonTextBlock_175 {
     id=#5
     possessed_class=CommonTextBlock
     tracks [
      {
       type=MovieScene2DTransformTrack
       sections [{
         type=MovieScene2DTransformSection
         channels [
          Translation.X {type=MovieSceneScriptingFloatChannel}
          Translation.Y {type=MovieSceneScriptingFloatChannel, const [0,1]}
         ]
        }]
      }
      {
       type=MovieSceneFloatTrack
       display_name=Render Opacity
       sections [{type=MovieSceneFloatSection, channels [None {type=MovieSceneScriptingFloatChannel, keys [[2,0],[4,1]]}]}]
      }
     ]
    }
    Text_Prerequisites {
     id=#7
     possessed_class=CommonTextBlock
     tracks [
      {
       type=MovieScene2DTransformTrack
       sections [{
         type=MovieScene2DTransformSection
         channels [
          Translation.X {type=MovieSceneScriptingFloatChannel}
          Translation.Y {type=MovieSceneScriptingFloatChannel, const [0,1]}
         ]
        }]
      }
      {
       type=MovieSceneFloatTrack
       display_name=Render Opacity
       sections [{type=MovieSceneFloatSection, channels [None {type=MovieSceneScriptingFloatChannel, keys [[0,0],[2,1]]}]}]
      }
     ]
    }
    QuestProgression_MainMission {
     id=#9
     possessed_class=SpatialStyleChallengeUnlockReason_C
     tracks [
      {
       type=MovieScene2DTransformTrack
       sections [{
         type=MovieScene2DTransformSection
         channels [
          Translation.X {type=MovieSceneScriptingFloatChannel}
          Translation.Y {type=MovieSceneScriptingFloatChannel, keys [[3,-24],[5,0]]}
         ]
        }]
      }
      {
       type=MovieSceneFloatTrack
       display_name=Render Opacity
       sections [{type=MovieSceneFloatSection, channels [None {type=MovieSceneScriptingFloatChannel, keys [[3,0],[5,1]]}]}]
      }
     ]
    }
   ]}
 }
]
graphs [
 EventGraph {nodes [
   n1 = Event OnPrerequisitesAndMissionUpdated!() -> (OutputDelegate, then, bLocked:bool, CharacterName:"Text (by ref)", StyleName:"Text (by ref)") {is=title K2Node_Event}
   > HandlePrerequisitesAndMissionUpdated!(execute, self, Locked:bool=@n1.bLocked, CharacterName:Text=@n1.CharacterName, StyleName:Text=@n1.StyleName) -> (then)
   Event OnCharacterUpdated() {is=title K2Node_Event}
   > PlayAnimationForward(InAnimation=Get PageRefresh() -> (PageRefresh:Widget Animation Object Reference))
   Event OnStyleUpdated() {is=title K2Node_Event}
   > PlayAnimationForward(InAnimation=Get StyleRefresh() -> (StyleRefresh:Widget Animation Object Reference))
  ]}
 HandlePrerequisitesAndMissionUpdated {nodes [
   HandlePrerequisitesAndMissionUpdated!() -> (then, Locked:bool=~, CharacterName:Text, StyleName:Text) {is=title K2Node_FunctionEntry}
   n1 = Get StyleName() -> (StyleName:Text)
   > Set StyleNameCorrected(StyleNameCorrected:Text=Select(Option 0:Text=Format Text!(Format="LOCTEXT(\"({Style})\")", Style:Text=@n1) -> (Result), Option 1:Text, Index=TextIsEmpty(InText=@n1)) -> (ReturnValue:Text)) -> (Output_Get:Text)
   n2 = Get CharacterName() -> (CharacterName:Text)
   n3 = Get StyleNameCorrected() -> (StyleNameCorrected:Text)
   n4 = Get CharacterName()
   n5 = Get StyleNameCorrected()
   > SetText(
    self=Get RichText_UnlockStyle() -> (RichText_UnlockStyle:Common Rich Text Block Object Reference)
    InText=Select(
     Option 0:"Text (by ref)"=Format Text!(Format="LOCTEXT(\"{Character} {Style} <text color=\\\"FFF\\\">UNLOCKED!</>\")", Character:Text=@n4, Style=@n5) -> (Result)
     Option 1:"Text (by ref)"=Format Text(Format="LOCTEXT(\"{Character} {Style}\")", Character=@n2, Style=@n3)
     Index=Get Locked() -> (Locked:bool)
    ) -> (ReturnValue:"Text (by ref)")
   )
  ]}
 PlayIntroAnim {nodes [
   PlayIntroAnim() {is=title K2Node_FunctionEntry}
   > PlayAnimationForward(InAnimation=Get Intro())
  ]}
]
variables []
widgets {
 info {
  inheritedWidgetCount=6
  parentClass=/Script/FortniteUI.AthenaSpatialStyleChallengeDetails
  rootWidgetClass=&Overlay
  widgetCount=25
 }
 widgets [
  Overlay_1 {widgetClassPath=&Overlay, widget=Overlay_1}
  SafeZone_0 {
   widgetClassPath=&SafeZone
   bIsVariable=true
   parent=Overlay_1
   widget=SafeZone_0
   properties {padBottom=false, padLeft=false, padTop=false}
   slot {class=OverlaySlot, properties {horizontalAlignment=HAlign_Right, verticalAlignment=VAlign_Fill}}
  }
  Backplate {
   widgetClassPath=&Image
   bIsVariable=true
   parent=SafeZone_0
   widget=Backplate
   properties {
    visibility=SelfHitTestInvisible
    brush {
     bIsDynamicallyLoaded=false
     drawAs=Image
     imageType=NoImage
     mirroring=NoMirror
     resourceName=None
     resourceObject=/BRCosmetics/Athena/UI/Frontend/Spatial/Art/M_UI_SpatialCusotomization_BackingPanel_Safezone
     tiling=NoTile
     imageSize=<1024 1024>
     margin=<ltrb 0.5 0 0 0>
     outlineSettings {
      bUseBrushTransparency=false
      roundingType=HalfHeightRadius
      width=0
      color {colorUseRule=UseColor_Specified, specifiedColor=<rgba 0 0 0 0>}
      cornerRadii=<0 0 0 0>
     }
     tintColor {colorUseRule=UseColor_Specified, specifiedColor=<rgba 1 1 1 1>}
     uVRegion {bIsValid=false, max=<0 0>, min=<0 0>}
    }
    colorAndOpacity=<rgba 1 1 1 0.85>
    renderTransform {angle=0, scale=<1.225 1>, shear=<0 0>, translation=<128 0>}
   }
   slot {class=SafeZoneSlot}
  }
  SafeZone_51 {
   widgetClassPath=&SafeZone
   bIsVariable=true
   parent=Overlay_1
   widget=SafeZone_51
   slot {class=OverlaySlot, properties {horizontalAlignment=HAlign_Right, verticalAlignment=VAlign_Fill}}
  }
  SizeBox_52 {
   widgetClassPath=&SizeBox
   parent=SafeZone_51
   widget=SizeBox_52
   properties {bOverride_WidthOverride=true, widthOverride=640}
   slot {class=SafeZoneSlot, properties {hAlign=HAlign_Right, padding=<ltrb 0 64 128 128>}}
  }
  VerticalBox_92 {widgetClassPath=&VerticalBox, parent=SizeBox_52, widget=VerticalBox_92, slot {class=SizeBoxSlot}}
  RichText_CompleteText {
   widgetClassPath=&CommonRichTextBlock
   bIsVariable=true
   parent=VerticalBox_92
   widget=RichText_CompleteText
   properties {
    bOverrideDefaultStyle=true
    justification=Center
    text="Complete <text color=\"FFF\"> MISSIONS</> to earn <text color=\"FFF\"> EACH REWARD! </>"
    wrapTextAt=600
    defaultTextStyleOverride {
     overflowPolicy=Clip
     transformPolicy=None
     colorAndOpacity {colorUseRule=UseColor_Specified, specifiedColor=<rgba 0 0.7682 0.9823 1>}
     font {
      bForceMonospaced=false
      bMaterialIsStencil=false
      fontMaterial=null
      fontObject=&BurbankSmall
      letterSpacing=0
      monospacedWidth=1
      size=18
      skewAmount=0
      typefaceFontName=Bold
      outlineSettings {
       bApplyOutlineToDropShadows=false
       bMiteredCorners=false
       bSeparateFillAlpha=false
       outlineMaterial=null
       outlineSize=0
       outlineColor=<rgba 0 0 0 1>
      }
     }
     highlightColor {colorUseRule=UseColor_Specified, specifiedColor=<rgba 0 0 0 1>}
     highlightShape {
      bIsDynamicallyLoaded=false
      drawAs=Image
      imageType=NoImage
      mirroring=NoMirror
      resourceName=None
      resourceObject=null
      tiling=NoTile
      imageSize=<32 32>
      margin=<ltrb 0 0 0 0>
      outlineSettings {
       bUseBrushTransparency=false
       roundingType=HalfHeightRadius
       width=0
       color {colorUseRule=UseColor_Specified, specifiedColor=<rgba 0 0 0 0>}
       cornerRadii=<0 0 0 0>
      }
      tintColor {colorUseRule=UseColor_Specified, specifiedColor=<rgba 1 1 1 1>}
      uVRegion {bIsValid=false, max=<0 0>, min=<0 0>}
     }
     selectedBackgroundColor {colorUseRule=UseColor_Foreground, specifiedColor=<rgba 1 0 1 1>}
     shadowColorAndOpacity=<rgba 0 0 0 1>
     shadowOffset=<0 0>
     strikeBrush {
      bIsDynamicallyLoaded=false
      drawAs=Image
      imageType=NoImage
      mirroring=NoMirror
      resourceName=None
      resourceObject=null
      tiling=NoTile
      imageSize=<32 32>
      margin=<ltrb 0 0 0 0>
      outlineSettings {
       bUseBrushTransparency=false
       roundingType=HalfHeightRadius
       width=0
       color {colorUseRule=UseColor_Specified, specifiedColor=<rgba 0 0 0 0>}
       cornerRadii=<0 0 0 0>
      }
      tintColor {colorUseRule=UseColor_Specified, specifiedColor=<rgba 1 1 1 1>}
      uVRegion {bIsValid=false, max=<0 0>, min=<0 0>}
     }
     underlineBrush {
      bIsDynamicallyLoaded=false
      drawAs=Image
      imageType=NoImage
      mirroring=NoMirror
      resourceName=None
      resourceObject=null
      tiling=NoTile
      imageSize=<32 32>
      margin=<ltrb 0 0 0 0>
      outlineSettings {
       bUseBrushTransparency=false
       roundingType=HalfHeightRadius
       width=0
       color {colorUseRule=UseColor_Specified, specifiedColor=<rgba 0 0 0 0>}
       cornerRadii=<0 0 0 0>
      }
      tintColor {colorUseRule=UseColor_Specified, specifiedColor=<rgba 1 1 1 1>}
      uVRegion {bIsValid=false, max=<0 0>, min=<0 0>}
     }
    }
    renderTransform {angle=0, scale=<1 1>, shear=<-8 0>, translation=<0 0>}
   }
   slot {class=VerticalBoxSlot, properties {horizontalAlignment=HAlign_Center, verticalAlignment=VAlign_Center}}
  }
  SizeBox_0 {
   widgetClassPath=&SizeBox
   parent=VerticalBox_92
   widget=SizeBox_0
   properties {bOverride_HeightOverride=true, heightOverride=192, maxDesiredHeight=180, minDesiredWidth=512}
   slot {class=VerticalBoxSlot, properties {horizontalAlignment=HAlign_Center, verticalAlignment=VAlign_Top, padding=<ltrb 0 32 0 0>}}
  }
  EntryBox_Styles {
   widgetClassPath=&DynamicEntryBox
   bInherited=true
   parent=SizeBox_0
   widget=EntryBox_Styles
   properties {
    entryVerticalAlignment=VAlign_Top
    entryWidgetClass=&/SpatialStyleCard.SpatialStyleCard_C
    maxElementSize=160
    entrySpacing=<24 0>
   }
   slot {class=SizeBoxSlot, properties {verticalAlignment=VAlign_Bottom}}
  }
  ScaleBox_0 {
   widgetClassPath=&ScaleBox
   parent=VerticalBox_92
   widget=ScaleBox_0
   properties {stretchDirection=DownOnly}
   slot {class=VerticalBoxSlot, properties {padding=<ltrb 0 32 0 0>}}
  }
  RichText_UnlockStyle {
   widgetClassPath=&CommonRichTextBlock
   bIsVariable=true
   parent=ScaleBox_0
   widget=RichText_UnlockStyle
   properties {
    bOverrideDefaultStyle=true
    text="Unlock Style ({Style})"
    textTransformPolicy=ToUpper
    defaultTextStyleOverride {
     overflowPolicy=Clip
     transformPolicy=None
     colorAndOpacity {colorUseRule=UseColor_Specified, specifiedColor=<rgba 0 0.7682 0.9823 1>}
     font {
      bForceMonospaced=false
      bMaterialIsStencil=false
      fontMaterial=null
      fontObject=&BurbankSmall
      letterSpacing=0
      monospacedWidth=1
      size=24
      skewAmount=0
      typefaceFontName=Bold
      outlineSettings {
       bApplyOutlineToDropShadows=false
       bMiteredCorners=false
       bSeparateFillAlpha=false
       outlineMaterial=null
       outlineSize=0
       outlineColor=<rgba 0 0 0 1>
      }
     }
     highlightColor {colorUseRule=UseColor_Specified, specifiedColor=<rgba 0 0 0 1>}
     highlightShape {
      bIsDynamicallyLoaded=false
      drawAs=Image
      imageType=NoImage
      mirroring=NoMirror
      resourceName=None
      resourceObject=null
      tiling=NoTile
      imageSize=<32 32>
      margin=<ltrb 0 0 0 0>
      outlineSettings {
       bUseBrushTransparency=false
       roundingType=HalfHeightRadius
       width=0
       color {colorUseRule=UseColor_Specified, specifiedColor=<rgba 0 0 0 0>}
       cornerRadii=<0 0 0 0>
      }
      tintColor {colorUseRule=UseColor_Specified, specifiedColor=<rgba 1 1 1 1>}
      uVRegion {bIsValid=false, max=<0 0>, min=<0 0>}
     }
     selectedBackgroundColor {colorUseRule=UseColor_Foreground, specifiedColor=<rgba 1 0 1 1>}
     shadowColorAndOpacity=<rgba 0 0 0 1>
     shadowOffset=<0 0>
     strikeBrush {
      bIsDynamicallyLoaded=false
      drawAs=Image
      imageType=NoImage
      mirroring=NoMirror
      resourceName=None
      resourceObject=null
      tiling=NoTile
      imageSize=<32 32>
      margin=<ltrb 0 0 0 0>
      outlineSettings {
       bUseBrushTransparency=false
       roundingType=HalfHeightRadius
       width=0
       color {colorUseRule=UseColor_Specified, specifiedColor=<rgba 0 0 0 0>}
       cornerRadii=<0 0 0 0>
      }
      tintColor {colorUseRule=UseColor_Specified, specifiedColor=<rgba 1 1 1 1>}
      uVRegion {bIsValid=false, max=<0 0>, min=<0 0>}
     }
     underlineBrush {
      bIsDynamicallyLoaded=false
      drawAs=Image
      imageType=NoImage
      mirroring=NoMirror
      resourceName=None
      resourceObject=null
      tiling=NoTile
      imageSize=<32 32>
      margin=<ltrb 0 0 0 0>
      outlineSettings {
       bUseBrushTransparency=false
       roundingType=HalfHeightRadius
       width=0
       color {colorUseRule=UseColor_Specified, specifiedColor=<rgba 0 0 0 0>}
       cornerRadii=<0 0 0 0>
      }
      tintColor {colorUseRule=UseColor_Specified, specifiedColor=<rgba 1 1 1 1>}
      uVRegion {bIsValid=false, max=<0 0>, min=<0 0>}
     }
    }
    renderTransform {angle=0, scale=<1 1>, shear=<-8 0>, translation=<0 0>}
   }
   slot {class=ScaleBoxSlot}
  }
  CommonWidgetSwitcher_Mission {
   widgetClassPath=&CommonWidgetSwitcherLegacy
   bIsVariable=true
   parent=VerticalBox_92
   widget=CommonWidgetSwitcher_Mission
   slot {class=VerticalBoxSlot, properties {padding=<ltrb 0 32 0 12>}}
  }
  VerticalBox_PrerequisitesandMissions {
   widgetClassPath=&VerticalBox
   parent=CommonWidgetSwitcher_Mission
   widget=VerticalBox_PrerequisitesandMissions
   slot {class=WidgetSwitcherSlot}
  }
  VBox_Prerequisities {
   widgetClassPath=&VerticalBox
   bInherited=true
   parent=VerticalBox_PrerequisitesandMissions
   widget=VBox_Prerequisities
   slot {class=VerticalBoxSlot}
  }
  Text_Prerequisites {
   widgetClassPath=&CommonTextBlock
   parent=VBox_Prerequisities
   widget=Text_Prerequisites
   properties {
    text=Prerequisites
    textTransformPolicy=ToUpper
    colorAndOpacity {colorUseRule=UseColor_Specified, specifiedColor=<rgba 0 0.2502 1 1>}
    font {
     bForceMonospaced=false
     bMaterialIsStencil=false
     fontMaterial=null
     fontObject=&BurbankSmall
     letterSpacing=0
     monospacedWidth=1
     size=22
     skewAmount=0
     typefaceFontName=Bold
     outlineSettings {
      bApplyOutlineToDropShadows=false
      bMiteredCorners=false
      bSeparateFillAlpha=false
      outlineMaterial=null
      outlineSize=0
      outlineColor=<rgba 0 0 0 1>
     }
    }
    renderTransform {angle=0, scale=<1 1>, shear=<-8 0>, translation=<0 0>}
    shadowOffset=<0 0>
   }
   slot {class=VerticalBoxSlot}
  }
  EntryBox_Prerequisites {
   widgetClassPath=&DynamicEntryBox
   bInherited=true
   bIsVariable=true
   parent=VBox_Prerequisities
   widget=EntryBox_Prerequisites
   properties {
    entryBoxType=Vertical
    entryWidgetClass=&/SpatialStyleChallengeUnlockReason.SpatialStyleChallengeUnlockReason_C
    entrySpacing=<0 24>
   }
   slot {class=VerticalBoxSlot, properties {padding=<ltrb 0 32 0 0>}}
  }
  VBox_MainMission {
   widgetClassPath=&VerticalBox
   bInherited=true
   parent=VerticalBox_PrerequisitesandMissions
   widget=VBox_MainMission
   slot {class=VerticalBoxSlot}
  }
  CommonTextBlock_175 {
   widgetClassPath=&CommonTextBlock
   parent=VBox_MainMission
   widget=CommonTextBlock_175
   properties {
    text=Main Mission
    textTransformPolicy=ToUpper
    colorAndOpacity {colorUseRule=UseColor_Specified, specifiedColor=<rgba 0 0.2502 1 1>}
    font {
     bForceMonospaced=false
     bMaterialIsStencil=false
     fontMaterial=null
     fontObject=&BurbankSmall
     letterSpacing=0
     monospacedWidth=1
     size=22
     skewAmount=0
     typefaceFontName=Bold
     outlineSettings {
      bApplyOutlineToDropShadows=false
      bMiteredCorners=false
      bSeparateFillAlpha=false
      outlineMaterial=null
      outlineSize=0
      outlineColor=<rgba 0 0 0 1>
     }
    }
    renderTransform {angle=0, scale=<1 1>, shear=<-8 0>, translation=<0 0>}
    shadowOffset=<0 0>
   }
   slot {class=VerticalBoxSlot, properties {padding=<ltrb 0 32 0 0>}}
  }
  QuestProgression_MainMission {
   widgetClassPath=&/SpatialStyleChallengeUnlockReason
   bInherited=true
   bIsVariable=true
   parent=VBox_MainMission
   widget=QuestProgression_MainMission
   slot {class=VerticalBoxSlot, properties {padding=<ltrb 0 32 0 0>}}
  }
  HorizontalBox_Complete {
   widgetClassPath=&HorizontalBox
   parent=CommonWidgetSwitcher_Mission
   widget=HorizontalBox_Complete
   slot {class=WidgetSwitcherSlot, properties {horizontalAlignment=HAlign_Left}}
  }
  Complete {
   widgetClassPath=&FortLazyImage
   bIsVariable=true
   parent=HorizontalBox_Complete
   widget=Complete
   properties {brush {
     bIsDynamicallyLoaded=false
     drawAs=Image
     imageType=NoImage
     mirroring=NoMirror
     resourceName=None
     resourceObject=/Game/Athena/UI/Challenges/Art/T_UI_ChallengeCheck_64
     tiling=NoTile
     imageSize=<64 64>
     margin=<ltrb 0 0 0 0>
     outlineSettings {
      bUseBrushTransparency=false
      roundingType=HalfHeightRadius
      width=0
      color {colorUseRule=UseColor_Specified, specifiedColor=<rgba 0 0 0 0>}
      cornerRadii=<0 0 0 0>
     }
     tintColor {colorUseRule=UseColor_Specified, specifiedColor=<rgba 1 1 1 1>}
     uVRegion {bIsValid=false, max=<0 0>, min=<0 0>}
    }}
   slot {class=HorizontalBoxSlot, properties {verticalAlignment=VAlign_Center}}
  }
  VerticalBox_71 {
   widgetClassPath=&VerticalBox
   parent=HorizontalBox_Complete
   widget=VerticalBox_71
   slot {class=HorizontalBoxSlot, properties {padding=<ltrb 12 0 0 0>}}
  }
  RichText_UnlockStyle_1 {
   widgetClassPath=&CommonRichTextBlock
   bIsVariable=true
   parent=VerticalBox_71
   widget=RichText_UnlockStyle_1
   properties {
    bOverrideDefaultStyle=true
    text=Mission
    textTransformPolicy=ToUpper
    defaultTextStyleOverride {
     overflowPolicy=Clip
     transformPolicy=None
     colorAndOpacity {colorUseRule=UseColor_Specified, specifiedColor=<rgba 1 1 1 1>}
     font {
      bForceMonospaced=false
      bMaterialIsStencil=false
      fontMaterial=null
      fontObject=&BurbankSmall
      letterSpacing=0
      monospacedWidth=1
      size=24
      skewAmount=0
      typefaceFontName=Bold
      outlineSettings {
       bApplyOutlineToDropShadows=false
       bMiteredCorners=false
       bSeparateFillAlpha=false
       outlineMaterial=null
       outlineSize=0
       outlineColor=<rgba 0 0 0 1>
      }
     }
     highlightColor {colorUseRule=UseColor_Specified, specifiedColor=<rgba 0 0 0 1>}
     highlightShape {
      bIsDynamicallyLoaded=false
      drawAs=Image
      imageType=NoImage
      mirroring=NoMirror
      resourceName=None
      resourceObject=null
      tiling=NoTile
      imageSize=<32 32>
      margin=<ltrb 0 0 0 0>
      outlineSettings {
       bUseBrushTransparency=false
       roundingType=HalfHeightRadius
       width=0
       color {colorUseRule=UseColor_Specified, specifiedColor=<rgba 0 0 0 0>}
       cornerRadii=<0 0 0 0>
      }
      tintColor {colorUseRule=UseColor_Specified, specifiedColor=<rgba 1 1 1 1>}
      uVRegion {bIsValid=false, max=<0 0>, min=<0 0>}
     }
     selectedBackgroundColor {colorUseRule=UseColor_Foreground, specifiedColor=<rgba 1 0 1 1>}
     shadowColorAndOpacity=<rgba 0 0 0 1>
     shadowOffset=<0 0>
     strikeBrush {
      bIsDynamicallyLoaded=false
      drawAs=Image
      imageType=NoImage
      mirroring=NoMirror
      resourceName=None
      resourceObject=null
      tiling=NoTile
      imageSize=<32 32>
      margin=<ltrb 0 0 0 0>
      outlineSettings {
       bUseBrushTransparency=false
       roundingType=HalfHeightRadius
       width=0
       color {colorUseRule=UseColor_Specified, specifiedColor=<rgba 0 0 0 0>}
       cornerRadii=<0 0 0 0>
      }
      tintColor {colorUseRule=UseColor_Specified, specifiedColor=<rgba 1 1 1 1>}
      uVRegion {bIsValid=false, max=<0 0>, min=<0 0>}
     }
     underlineBrush {
      bIsDynamicallyLoaded=false
      drawAs=Image
      imageType=NoImage
      mirroring=NoMirror
      resourceName=None
      resourceObject=null
      tiling=NoTile
      imageSize=<32 32>
      margin=<ltrb 0 0 0 0>
      outlineSettings {
       bUseBrushTransparency=false
       roundingType=HalfHeightRadius
       width=0
       color {colorUseRule=UseColor_Specified, specifiedColor=<rgba 0 0 0 0>}
       cornerRadii=<0 0 0 0>
      }
      tintColor {colorUseRule=UseColor_Specified, specifiedColor=<rgba 1 1 1 1>}
      uVRegion {bIsValid=false, max=<0 0>, min=<0 0>}
     }
    }
    renderTransform {angle=0, scale=<1 1>, shear=<-8 0>, translation=<0 0>}
   }
   slot {class=VerticalBoxSlot, properties {horizontalAlignment=HAlign_Left, verticalAlignment=VAlign_Center, padding=<ltrb 8 0 0 0>}}
  }
  RichText_UnlockStyle_2 {
   widgetClassPath=&CommonRichTextBlock
   bIsVariable=true
   parent=VerticalBox_71
   widget=RichText_UnlockStyle_2
   properties {
    bOverrideDefaultStyle=true
    text=Complete
    textTransformPolicy=ToUpper
    defaultTextStyleOverride {
     overflowPolicy=Clip
     transformPolicy=None
     colorAndOpacity {colorUseRule=UseColor_Specified, specifiedColor=<rgba 0.0331 1 0.2307 1>}
     font {
      bForceMonospaced=false
      bMaterialIsStencil=false
      fontMaterial=null
      fontObject=&BurbankSmall
      letterSpacing=0
      monospacedWidth=1
      size=36
      skewAmount=0
      typefaceFontName=Black
      outlineSettings {
       bApplyOutlineToDropShadows=false
       bMiteredCorners=false
       bSeparateFillAlpha=false
       outlineMaterial=null
       outlineSize=0
       outlineColor=<rgba 0 0 0 1>
      }
     }
     highlightColor {colorUseRule=UseColor_Specified, specifiedColor=<rgba 0 0 0 1>}
     highlightShape {
      bIsDynamicallyLoaded=false
      drawAs=Image
      imageType=NoImage
      mirroring=NoMirror
      resourceName=None
      resourceObject=null
      tiling=NoTile
      imageSize=<32 32>
      margin=<ltrb 0 0 0 0>
      outlineSettings {
       bUseBrushTransparency=false
       roundingType=HalfHeightRadius
       width=0
       color {colorUseRule=UseColor_Specified, specifiedColor=<rgba 0 0 0 0>}
       cornerRadii=<0 0 0 0>
      }
      tintColor {colorUseRule=UseColor_Specified, specifiedColor=<rgba 1 1 1 1>}
      uVRegion {bIsValid=false, max=<0 0>, min=<0 0>}
     }
     selectedBackgroundColor {colorUseRule=UseColor_Foreground, specifiedColor=<rgba 1 0 1 1>}
     shadowColorAndOpacity=<rgba 0 0 0 1>
     shadowOffset=<0 0>
     strikeBrush {
      bIsDynamicallyLoaded=false
      drawAs=Image
      imageType=NoImage
      mirroring=NoMirror
      resourceName=None
      resourceObject=null
      tiling=NoTile
      imageSize=<32 32>
      margin=<ltrb 0 0 0 0>
      outlineSettings {
       bUseBrushTransparency=false
       roundingType=HalfHeightRadius
       width=0
       color {colorUseRule=UseColor_Specified, specifiedColor=<rgba 0 0 0 0>}
       cornerRadii=<0 0 0 0>
      }
      tintColor {colorUseRule=UseColor_Specified, specifiedColor=<rgba 1 1 1 1>}
      uVRegion {bIsValid=false, max=<0 0>, min=<0 0>}
     }
     underlineBrush {
      bIsDynamicallyLoaded=false
      drawAs=Image
      imageType=NoImage
      mirroring=NoMirror
      resourceName=None
      resourceObject=null
      tiling=NoTile
      imageSize=<32 32>
      margin=<ltrb 0 0 0 0>
      outlineSettings {
       bUseBrushTransparency=false
       roundingType=HalfHeightRadius
       width=0
       color {colorUseRule=UseColor_Specified, specifiedColor=<rgba 0 0 0 0>}
       cornerRadii=<0 0 0 0>
      }
      tintColor {colorUseRule=UseColor_Specified, specifiedColor=<rgba 1 1 1 1>}
      uVRegion {bIsValid=false, max=<0 0>, min=<0 0>}
     }
    }
    renderTransform {angle=0, scale=<1 1>, shear=<-8 0>, translation=<0 0>}
   }
   slot {class=VerticalBoxSlot, properties {horizontalAlignment=HAlign_Left, verticalAlignment=VAlign_Center, padding=<ltrb 0 -6 0 0>}}
  }
  Button_PurchaseBattlepass {
   widgetClassPath=/FortniteCoreUI/Frontend/Spatial/Customization/Spatial_Specialbutton
   bInherited=true
   bIsVariable=true
   parent=VerticalBox_92
   widget=Button_PurchaseBattlepass
   properties {
    button text="Purchase\r\n<text color=\"ffffffff\">Battle Pass!</>"
    justifiction=Center
    onButtonBaseClicked=(&self:WidgetTree.Button_PurchaseBattlepass.HandleOnButtonClicked)
    onButtonBaseDoubleClicked=(&self:WidgetTree.Button_PurchaseBattlepass.HandleOnButtonDoubleClicked)
    onButtonBaseHovered=(&self:WidgetTree.Button_PurchaseBattlepass.HandleOnButtonHovered)
    onButtonBaseUnhovered=(&self:WidgetTree.Button_PurchaseBattlepass.HandleOnButtonUnhovered)
    onSelectedChangedBase=(&self:WidgetTree.Button_PurchaseBattlepass.HandleOnSelectedChanged)
    text Color {colorUseRule=UseColor_Specified, specifiedColor=<rgba 0 0.3333 1 1>}
    triggeringInputAction {dataTable=&InputActionDataTable, rowName=GenericAccept}
   }
   slot {class=VerticalBoxSlot, properties {verticalAlignment=VAlign_Top, padding=<ltrb 0 0 10 0>}}
  }
 ]
}
