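# BattlePassPagePipButton is a CommonButtonLegacy widget that draws a page pip, apparently for the Battle Pass track.
# A single image uses the M_UI_BattlePassTrack_PagePip material, with two short animations driving its scalar parameters.
# Hover animates backing color, playing forward faster on hover and reversing on unhover.
# Pressed animates selected color plus shape offsets and rotations, playing on click or selection and reversing on deselection.
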WidgetBlueprint /BRCosmetics/Athena/UI/Season/BattlePassPagePipButton
animations [
 Hover {
  binding_count=1
  length_seconds=0.15
  object=&self:Hover
  sequence {bindings [ButtonBacking {
     id=#1
     possessed_class=Image
     tracks [{
       type=MovieSceneWidgetMaterialTrack
       sections [{type=MovieSceneParameterSection, channels [Animate_BackingColor {type=MovieSceneScriptingFloatChannel, keys [[0,0],[3,1]]}]}]
      }]
    }]}
 }
 Pressed {
  binding_count=1
  length_seconds=0.15
  object=&self:Pressed
  sequence {bindings [ButtonBacking {
     id=#2
     possessed_class=Image
     tracks [{
       type=MovieSceneWidgetMaterialTrack
       sections [{
         type=MovieSceneParameterSection
         channels [
          Animate_BackingColor_Selected {type=MovieSceneScriptingFloatChannel, keys [[0,0],[3,1]]}
          Shape_Offset_Bottom {type=MovieSceneScriptingFloatChannel, keys [[0,-0.25],[3,-0.125]]}
          Shape_Offset_Left {type=MovieSceneScriptingFloatChannel, keys [[0,-0.25],[3,-0.125]]}
          Shape_Offset_Right {type=MovieSceneScriptingFloatChannel, keys [[0,-0.25],[3,-0.125]]}
          Shape_Offset_Top {type=MovieSceneScriptingFloatChannel, keys [[0,-0.25],[3,-0.125]]}
          Shape_Rotation_Bottom {type=MovieSceneScriptingFloatChannel, keys [[0,-0.245],[3,-0.255]]}
          Shape_Rotation_Top {type=MovieSceneScriptingFloatChannel, keys [[0,-0.245],[3,-0.235]]}
          Shape_Rotation_Left {type=MovieSceneScriptingFloatChannel, keys [[0,-0.025],[3,-0.005]]}
          Shape_Rotation_Right {type=MovieSceneScriptingFloatChannel, keys [[0,-0.025],[3,-0.015]]}
         ]
        }]
      }]
    }]}
 }
]
graphs [EventGraph {nodes [
   n1 = Event On Selected()
   PlayAnimation(execute=[@n1.then, @n3.then], InAnimation=Get Pressed() -> (Pressed:Widget Animation Object Reference))
   Event On Deselected()
   > PlayAnimation(InAnimation=Get Pressed(), PlayMode=Reverse)
   Event On Hovered()
   n2 = Get Hover() -> (Hover:Widget Animation Object Reference)
   > PlayAnimationForward(InAnimation=@n2, PlaybackSpeed=1.5)
   Event On Unhovered()
   > PlayAnimationReverse(InAnimation=@n2)
   n3 = Event On Clicked()
   Event PreConstruct()
   Event Construct()
   Event Tick()
  ]}]
variables []
widgets {
 info {inheritedWidgetCount=1, parentClass=&CommonButtonLegacy, rootWidgetClass=&Overlay, widgetCount=3}
 widgets [
  Overlay_29 {widgetClassPath=&Overlay, widget=Overlay_29}
  ButtonBacking {
   widgetClassPath=&Image
   bIsVariable=true
   parent=Overlay_29
   widget=ButtonBacking
   properties {brush {
     bIsDynamicallyLoaded=false
     drawAs=Image
     imageType=NoImage
     mirroring=NoMirror
     resourceName=None
     resourceObject=/Game/Athena/UI/Season/Art/M_UI_BattlePassTrack_PagePip
     tiling=NoTile
     imageSize=<24 24>
     margin=<ltrb 0 0 0 0>
     outlineSettings {
      bUseBrushTransparency=false
      roundingType=HalfHeightRadius
      width=0
      color {colorUseRule=UseColor_Specified, specifiedColor=<rgba 0 0 0 0>}
      cornerRadii=<0 0 0 0>
     }
     tintColor {colorUseRule=UseColor_Specified, specifiedColor=<rgba 1 1 1 1>}
     uVRegion {bIsValid=false, max=<0 0>, min=<0 0>}
    }}
   slot {class=OverlaySlot, properties {horizontalAlignment=HAlign_Fill, verticalAlignment=VAlign_Fill}}
  }
  InputActionWidget {widgetClassPath=&CommonActionWidget, bInherited=true, widget=null}
 ]
}
