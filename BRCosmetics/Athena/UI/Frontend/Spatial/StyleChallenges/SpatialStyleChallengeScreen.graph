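# Battle Pass Style Challenges frontend screen, a widget Blueprint on AthenaSpatialStyleChallengeScreen for browsing characters and their challenge details.
# Previous/next shoulder-bound buttons, a pip entry box and title text sit beside the details widget, with only a title/description animation wired here to the arrows.
# More Info and first-time info fire a MoreInfo analytic and switch to an About modal.
# Activation and modal close replay intro animations; closing also restores main content.
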WidgetBlueprint /BRCosmetics/Athena/UI/Frontend/Spatial/StyleChallenges/SpatialStyleChallengeScreen
animations [
 Intro {
  binding_count=6
  length_seconds=0.25
  object=&self:Intro
  sequence {bindings [
    Text_Title {
     id=#1
     possessed_class=CommonTextBlock
     tracks [
      {
       type=MovieScene2DTransformTrack
       sections [{type=MovieScene2DTransformSection, channels [Translation.Y {type=MovieSceneScriptingFloatChannel, keys [[0,48],[5,0]]}]}]
      }
      {
       type=MovieSceneFloatTrack
       display_name=Render Opacity
       sections [{type=MovieSceneFloatSection, channels [None {type=MovieSceneScriptingFloatChannel, keys [[0,0],[5,1]]}]}]
      }
     ]
    }
    FakeHackGradient {
     id=#2
     possessed_class=Image
     tracks [
      {
       type=MovieSceneFloatTrack
       display_name=Render Opacity
       sections [{type=MovieSceneFloatSection, channels [None {type=MovieSceneScriptingFloatChannel, keys [[0,0],[3,1]]}]}]
      }
      {
       type=MovieScene2DTransformTrack
       sections [{type=MovieScene2DTransformSection, channels [Translation.Y {type=MovieSceneScriptingFloatChannel, keys [[0,256],[3,0]]}]}]
      }
      {
       type=MovieSceneWidgetMaterialTrack
       sections [{type=MovieSceneParameterSection, channels [OverallOpacity {type=MovieSceneScriptingFloatChannel, keys [[0,0.5],[3,0.75]]}]}]
      }
     ]
    }
    Button_Next {
     id=#3
     possessed_class=CircleArrowButton_C
     tracks [
      {
       type=MovieScene2DTransformTrack
       sections [{type=MovieScene2DTransformSection, channels [Translation.X {type=MovieSceneScriptingFloatChannel, keys [[0,-64],[5.0003,0]]}]}]
      }
      {
       type=MovieSceneFloatTrack
       display_name=Render Opacity
       sections [{type=MovieSceneFloatSection, channels [None {type=MovieSceneScriptingFloatChannel, keys [[0,0],[5,1]]}]}]
      }
     ]
    }
    Button_Previous {
     id=#4
     possessed_class=CircleArrowButton_C
     tracks [
      {
       type=MovieScene2DTransformTrack
       sections [{type=MovieScene2DTransformSection, channels [Translation.X {type=MovieSceneScriptingFloatChannel, keys [[0,64],[5.0003,0]]}]}]
      }
      {
       type=MovieSceneFloatTrack
       display_name=Render Opacity
       sections [{type=MovieSceneFloatSection, channels [None {type=MovieSceneScriptingFloatChannel, keys [[0,0],[5.0003,1]]}]}]
      }
     ]
    }
    Text_Description {
     id=#5
     possessed_class=CommonTextBlock
     tracks [{
       type=MovieSceneFloatTrack
       display_name=Render Opacity
       sections [{type=MovieSceneFloatSection, channels [None {type=MovieSceneScriptingFloatChannel, keys [[0,0],[5,1]]}]}]
      }]
    }
    EntryBox_CharacterChoices {
     id=#6
     possessed_class=DynamicEntryBox
     tracks [
      {
       type=MovieSceneFloatTrack
       display_name=Render Opacity
       sections [{type=MovieSceneFloatSection, channels [None {type=MovieSceneScriptingFloatChannel, keys [[0,0],[5,1]]}]}]
      }
      {
       type=MovieScene2DTransformTrack
       sections [{type=MovieScene2DTransformSection, channels [Translation.Y {type=MovieSceneScriptingFloatChannel, keys [[0,-8],[5,0]]}]}]
      }
     ]
    }
   ]}
 }
 Switch-Character {
  binding_count=2
  length_seconds=0.25
  object=&self:Switch-Character
  sequence {bindings [
    Text_Title {
     id=#1
     possessed_class=CommonTextBlock
     tracks [
      {
       type=MovieScene2DTransformTrack
       sections [{type=MovieScene2DTransformSection, channels [Translation.Y {type=MovieSceneScriptingFloatChannel, keys [[0,48],[5,0]]}]}]
      }
      {
       type=MovieSceneFloatTrack
       display_name=Render Opacity
       sections [{type=MovieSceneFloatSection, channels [None {type=MovieSceneScriptingFloatChannel, keys [[0,0],[5,1]]}]}]
      }
     ]
    }
    Text_Description {
     id=#5
     possessed_class=CommonTextBlock
     tracks [{
       type=MovieSceneFloatTrack
       display_name=Render Opacity
       sections [{type=MovieSceneFloatSection, channels [None {type=MovieSceneScriptingFloatChannel, keys [[0,0],[5,1]]}]}]
      }]
    }
   ]}
 }
]
graphs [EventGraph {nodes [
   n1 = "On Button Clicked (Button_Next)"!() -> (OutputDelegate, then, Button:Common Button Legacy Object Reference) {is=title K2Node_ComponentBoundEvent}
   PlayAnimationForward(execute=[@n1.then, @n2.then], InAnimation=Get Switch-Character() -> (Switch-Character:Widget Animation Object Reference))
   n2 = "On Button Clicked (Button_Previous)"!() -> (OutputDelegate, then, Button:Common Button Legacy Object Reference) {is=title K2Node_ComponentBoundEvent}
   "BP on About Modal Closed (BPS12_AboutModal)"() {is=title K2Node_ComponentBoundEvent}
   > SetActiveWidget(
    self:Widget Switcher Object Reference=Get Switcher_Content() -> (Switcher_Content:Common Activatable Widget Switcher Object Reference)
    Widget=Get Overlay_ChallengeScreenContent() -> (Overlay_ChallengeScreenContent:Overlay Object Reference)
   )
   > PlayAnimationForward(InAnimation=Get Intro())
   > n3 = Cast To SpatialStyleChallengeDetails!(
    execute
    Object=Get ChallengeDetails_CharacterDetails() -> (ChallengeDetails_CharacterDetails:Athena Spatial Style Challenge Details Object Reference)
   ) -> (then, CastFailed, AsSpatial Style Challenge Details:Spatial Style Challenge Details Object Reference) {is=title K2Node_DynamicCast}
   > PlayIntroAnim(self:Spatial Style Challenge Details Object Reference=@n3)
   n4 = "On Button Clicked (Button_About)"!() -> (OutputDelegate, then, Button:Common Button Legacy Object Reference) {is=title K2Node_ComponentBoundEvent}
   FireSpatialScreenAnalytic!(execute=[@n4.then, @n6.then], self:Athena Spatial Screen Object Reference, Interaction:String=MoreInfo) -> (then)
   > SetActiveWidget(self=Get Switcher_Content(), Widget=Get BPS12_AboutModal() -> (BPS12_AboutModal:BPS12 About Modal Object Reference))
   Event On Activated()
   > PlayAnimationForward(InAnimation=Get Intro())
   > n5 = Cast To SpatialStyleChallengeDetails(Object=Get ChallengeDetails_CharacterDetails())
   > PlayIntroAnim(self=@n5)
   n6 = Event OnShowFirstTimeInfo() {is=title K2Node_Event}
  ]}]
variables []
widgets {
 info {
  inheritedWidgetCount=8
  parentClass=/Script/FortniteUI.AthenaSpatialStyleChallengeScreen
  rootWidgetClass=&CommonActivatableWidgetSwitcher
  widgetCount=26
 }
 widgets [
  Switcher_Content {widgetClassPath=&CommonActivatableWidgetSwitcher, bIsVariable=true, widget=Switcher_Content}
  Overlay_ChallengeScreenContent {
   widgetClassPath=&Overlay
   bIsVariable=true
   parent=Switcher_Content
   widget=Overlay_ChallengeScreenContent
   slot {class=WidgetSwitcherSlot}
  }
  FakeHackGradient {
   widgetClassPath=&Image
   bIsVariable=true
   parent=Overlay_ChallengeScreenContent
   widget=FakeHackGradient
   properties {
    visibility=SelfHitTestInvisible
    brush {
     bIsDynamicallyLoaded=false
     drawAs=Image
     imageType=NoImage
     mirroring=NoMirror
     resourceName=None
     resourceObject=/BRCosmetics/UI/Art_ToolKit/Tools/M_UI_Bp_HackGradient_Inst
     tiling=NoTile
     imageSize=<32 704>
     margin=<ltrb 0 0 0 0>
     outlineSettings {
      bUseBrushTransparency=false
      roundingType=HalfHeightRadius
      width=0
      color {colorUseRule=UseColor_Specified, specifiedColor=<rgba 0 0 0 0>}
      cornerRadii=<0 0 0 0>
     }
     tintColor {colorUseRule=UseColor_Specified, specifiedColor=<rgba 1 1 1 1>}
     uVRegion {bIsValid=false, max=<0 0>, min=<0 0>}
    }
   }
   slot {class=OverlaySlot, properties {horizontalAlignment=HAlign_Fill, verticalAlignment=VAlign_Bottom}}
  }
  HorizontalBox_0 {
   widgetClassPath=&HorizontalBox
   parent=Overlay_ChallengeScreenContent
   widget=HorizontalBox_0
   slot {class=OverlaySlot, properties {horizontalAlignment=HAlign_Fill, verticalAlignment=VAlign_Fill}}
  }
  SafeZone_CharacterChoice {
   widgetClassPath=&SafeZone
   bIsVariable=true
   parent=HorizontalBox_0
   widget=SafeZone_CharacterChoice
   slot {
    class=HorizontalBoxSlot
    properties {horizontalAlignment=HAlign_Right, verticalAlignment=VAlign_Bottom, padding=<ltrb 32 0 0 32>, size {sizeRule=Fill, value=0.5}}
   }
  }
  SizeBox_3 {
   widgetClassPath=&SizeBox
   parent=SafeZone_CharacterChoice
   widget=SizeBox_3
   properties {bOverride_WidthOverride=true, widthOverride=800}
   slot {class=SafeZoneSlot, properties {hAlign=HAlign_Center}}
  }
  HorizontalBox_203 {widgetClassPath=&HorizontalBox, parent=SizeBox_3, widget=HorizontalBox_203, slot {class=SizeBoxSlot}}
  SizeBox_1 {
   widgetClassPath=&SizeBox
   parent=HorizontalBox_203
   widget=SizeBox_1
   properties {bOverride_HeightOverride=true, bOverride_WidthOverride=true, heightOverride=64, widthOverride=64}
   slot {class=HorizontalBoxSlot, properties {horizontalAlignment=HAlign_Left, verticalAlignment=VAlign_Bottom, padding=<ltrb 0 0 16 24>}}
  }
  Button_Previous {
   widgetClassPath=/FortniteCoreUI/Frontend/Showdown/CircleArrowButton
   bInherited=true
   bIsVariable=true
   parent=SizeBox_1
   widget=Button_Previous
   properties {
    clickMethod=DownAndUp
    edgeSoftness=0.075
    flip=true
    onButtonBaseClicked=(&self:WidgetTree.Button_Previous.HandleOnButtonClicked)
    onButtonBaseDoubleClicked=(&self:WidgetTree.Button_Previous.HandleOnButtonDoubleClicked)
    onButtonBaseHovered=(&self:WidgetTree.Button_Previous.HandleOnButtonHovered)
    onButtonBaseUnhovered=(&self:WidgetTree.Button_Previous.HandleOnButtonUnhovered)
    onSelectedChangedBase=(&self:WidgetTree.Button_Previous.HandleOnSelectedChanged)
    pressMethod=DownAndUp
    useShadow=false
    visibility=Visible
    arrowColor=<rgba 0 0.3333 1 1>
    hoveredSlateSoundOverride {resourceObject=/Game/Sounds/Fort_UI_Sounds/S13_Frontend/UI_Battlepass_Main_Hover_Cue}
    pressedSlateSoundOverride {resourceObject=/BRCosmetics/Sounds/Other/Fort_UI_Sounds/S13_Frontend/UI_Battlepass_Main_Click_Cue}
    triggeringInputAction {dataTable=&InputActionDataTable, rowName=GenericLeftShoulder}
   }
   slot {class=SizeBoxSlot}
  }
  VerticalBox_109 {
   widgetClassPath=&VerticalBox
   parent=HorizontalBox_203
   widget=VerticalBox_109
   slot {class=HorizontalBoxSlot, properties {size {sizeRule=Fill, value=1}}}
  }
  SizeBox_2 {
   widgetClassPath=&SizeBox
   parent=VerticalBox_109
   widget=SizeBox_2
   properties {bOverride_MaxDesiredWidth=true, bOverride_MinDesiredHeight=true, maxDesiredWidth=680, minDesiredHeight=112, widthOverride=680}
   slot {class=VerticalBoxSlot, properties {horizontalAlignment=HAlign_Center}}
  }
  ScaleBox_0 {
   widgetClassPath=&ScaleBox
   parent=SizeBox_2
   widget=ScaleBox_0
   properties {clipping=ClipToBoundsAlways, stretchDirection=DownOnly}
   slot {class=SizeBoxSlot, properties {padding=<ltrb 0 0 0 8>}}
  }
  Text_Title {
   widgetClassPath=&CommonTextBlock
   bInherited=true
   parent=ScaleBox_0
   widget=Text_Title
   properties {
    justification=Center
    text="{Character-Name}"
    textTransformPolicy=ToUpper
    font {
     bForceMonospaced=false
     bMaterialIsStencil=false
     fontMaterial=null
     fontObject=&Burbank
     letterSpacing=0
     monospacedWidth=1
     size=64
     skewAmount=0
     typefaceFontName=Burbank-Black
     outlineSettings {
      bApplyOutlineToDropShadows=false
      bMiteredCorners=false
      bSeparateFillAlpha=true
      outlineMaterial=null
      outlineSize=0
      outlineColor=<rgba 0 0 0 1>
     }
    }
    margin=<ltrb 18 0 18 0>
    renderTransform {angle=0, scale=<1 1>, shear=<-8 0>, translation=<0 0>}
    shadowOffset=<0 0>
   }
   slot {class=ScaleBoxSlot, properties {verticalAlignment=VAlign_Bottom}}
  }
  Text_Description {
   widgetClassPath=&CommonTextBlock
   bInherited=true
   parent=VerticalBox_109
   widget=Text_Description
   properties {
    justification=Center
    text=Text Block
    textTransformPolicy=ToUpper
    wrapTextAt=580
    colorAndOpacity {colorUseRule=UseColor_Specified, specifiedColor=<rgba 0 0.7758 0.9911 1>}
    font {
     bForceMonospaced=false
     bMaterialIsStencil=false
     fontMaterial=null
     fontObject=&BurbankSmall
     letterSpacing=0
     monospacedWidth=1
     size=18
     skewAmount=0
     typefaceFontName=Black
     outlineSettings {
      bApplyOutlineToDropShadows=false
      bMiteredCorners=false
      bSeparateFillAlpha=false
      outlineMaterial=null
      outlineSize=0
      outlineColor=<rgba 0 0 0 1>
     }
    }
    renderTransform {angle=0, scale=<1 1>, shear=<-8 0>, translation=<0 0>}
    shadowOffset=<0 0>
   }
   slot {class=VerticalBoxSlot, properties {horizontalAlignment=HAlign_Center, verticalAlignment=VAlign_Top, padding=<ltrb 0 0 0 16>}}
  }
  EntryBox_CharacterChoices {
   widgetClassPath=&DynamicEntryBox
   bInherited=true
   bIsVariable=true
   parent=VerticalBox_109
   widget=EntryBox_CharacterChoices
   properties {entryWidgetClass=/BRCosmetics/Athena/UI/Season/BattlePassPagePipButton.BattlePassPagePipButton_C}
   slot {class=VerticalBoxSlot, properties {horizontalAlignment=HAlign_Center}}
  }
  SizeBox_0 {
   widgetClassPath=&SizeBox
   parent=HorizontalBox_203
   widget=SizeBox_0
   properties {bOverride_HeightOverride=true, bOverride_WidthOverride=true, heightOverride=64, widthOverride=64}
   slot {class=HorizontalBoxSlot, properties {horizontalAlignment=HAlign_Right, verticalAlignment=VAlign_Bottom, padding=<ltrb 16 0 0 24>}}
  }
  Button_Next {
   widgetClassPath=&CircleArrowButton
   bInherited=true
   bIsVariable=true
   parent=SizeBox_0
   widget=Button_Next
   properties {
    clickMethod=DownAndUp
    edgeSoftness=0.075
    onButtonBaseClicked=(&self:WidgetTree.Button_Next.HandleOnButtonClicked)
    onButtonBaseDoubleClicked=(&self:WidgetTree.Button_Next.HandleOnButtonDoubleClicked)
    onButtonBaseHovered=(&self:WidgetTree.Button_Next.HandleOnButtonHovered)
    onButtonBaseUnhovered=(&self:WidgetTree.Button_Next.HandleOnButtonUnhovered)
    onSelectedChangedBase=(&self:WidgetTree.Button_Next.HandleOnSelectedChanged)
    pressMethod=DownAndUp
    useShadow=false
    arrowColor=<rgba 0 0.3333 1 1>
    hoveredSlateSoundOverride {resourceObject=&UI_Battlepass_Main_Hover_Cue}
    pressedSlateSoundOverride {resourceObject=&UI_Battlepass_Main_Click_Cue}
    triggeringInputAction {dataTable=&InputActionDataTable, rowName=GenericRightShoulder}
   }
   slot {class=SizeBoxSlot}
  }
  ChallengeDetails_CharacterDetails {
   widgetClassPath=&/SpatialStyleChallengeDetails
   bInherited=true
   bIsVariable=true
   parent=HorizontalBox_0
   widget=ChallengeDetails_CharacterDetails
   slot {class=HorizontalBoxSlot, properties {size {sizeRule=Fill, value=0.55}}}
  }
  SafeZone_ScreenButtons {
   widgetClassPath=&SafeZone
   bIsVariable=true
   parent=Overlay_ChallengeScreenContent
   widget=SafeZone_ScreenButtons
   properties {padLeft=false, padTop=false}
   slot {class=OverlaySlot, properties {horizontalAlignment=HAlign_Right, verticalAlignment=VAlign_Bottom, padding=<ltrb 0 0 32 32>}}
  }
  HorizontalBox_113 {widgetClassPath=&HorizontalBox, parent=SafeZone_ScreenButtons, widget=HorizontalBox_113, slot {class=SafeZoneSlot}}
  Button_About {
   widgetClassPath=/FortUILibrary/Foundation/Buttons/SoloButton
   bIsVariable=true
   parent=HorizontalBox_113
   widget=Button_About
   properties {
    button Description=More Info
    onButtonBaseClicked=(&self:WidgetTree.Button_About.HandleOnButtonClicked)
    onButtonBaseDoubleClicked=(&self:WidgetTree.Button_About.HandleOnButtonDoubleClicked)
    onButtonBaseHovered=(&self:WidgetTree.Button_About.HandleOnButtonHovered)
    onButtonBaseUnhovered=(&self:WidgetTree.Button_About.HandleOnButtonUnhovered)
    onSelectedChangedBase=(&self:WidgetTree.Button_About.HandleOnSelectedChanged)
    triggeringInputAction {dataTable=&InputActionDataTable, rowName=GenericFaceButtonTop}
   }
   slot {class=HorizontalBoxSlot}
  }
  HardwareVisibilityBorder_Back {
   widgetClassPath=&CommonHardwareVisibilityBorder
   parent=HorizontalBox_113
   widget=HardwareVisibilityBorder_Back
   properties {visibilityQuery {
     autoDescription=" NONE( Input.Touch )"
     tokenStreamVersion=0
     userDescription=""
     queryTokenStream [0,1,3,1,0]
     tagDictionary [{tagName=Input.Touch}]
    }}
   slot {class=HorizontalBoxSlot, properties {padding=<ltrb 24 0 0 0>}}
  }
  Button_Back {
   widgetClassPath=&SoloButton
   bInherited=true
   parent=HardwareVisibilityBorder_Back
   widget=Button_Back
   properties {
    button Description=Back
    onButtonBaseClicked=(&self:WidgetTree.Button_Back.HandleOnButtonClicked)
    onButtonBaseDoubleClicked=(&self:WidgetTree.Button_Back.HandleOnButtonDoubleClicked)
    onButtonBaseHovered=(&self:WidgetTree.Button_Back.HandleOnButtonHovered)
    onButtonBaseUnhovered=(&self:WidgetTree.Button_Back.HandleOnButtonUnhovered)
    onSelectedChangedBase=(&self:WidgetTree.Button_Back.HandleOnSelectedChanged)
    triggeringInputAction {dataTable=&InputActionDataTable, rowName=GenericBack}
   }
   slot {class=BorderSlot, properties {padding=<ltrb 0 0 0 0>}}
  }
  HardwareVisibilityBorder_TouchBack {
   widgetClassPath=&CommonHardwareVisibilityBorder
   parent=Overlay_ChallengeScreenContent
   widget=HardwareVisibilityBorder_TouchBack
   properties {visibilityQuery {
     autoDescription=" ANY( Input.Touch )"
     tokenStreamVersion=0
     userDescription=""
     queryTokenStream [0,1,1,1,0]
     tagDictionary [{tagName=Input.Touch}]
    }}
   slot {class=OverlaySlot, properties {horizontalAlignment=HAlign_Right, padding=<ltrb 0 32 32 0>}}
  }
  Button_TouchBack {
   widgetClassPath=/FortUILibrary/Foundation/Buttons/CloseButton
   bInherited=true
   bIsVariable=true
   parent=HardwareVisibilityBorder_TouchBack
   widget=Button_TouchBack
   properties {
    onButtonBaseClicked=(&self:WidgetTree.Button_TouchBack.HandleOnButtonClicked)
    onButtonBaseDoubleClicked=(&self:WidgetTree.Button_TouchBack.HandleOnButtonDoubleClicked)
    onButtonBaseHovered=(&self:WidgetTree.Button_TouchBack.HandleOnButtonHovered)
    onButtonBaseUnhovered=(&self:WidgetTree.Button_TouchBack.HandleOnButtonUnhovered)
    onSelectedChangedBase=(&self:WidgetTree.Button_TouchBack.HandleOnSelectedChanged)
   }
   slot {class=BorderSlot, properties {padding=<ltrb 0 0 0 0>}}
  }
  BPS12_AboutModal {
   widgetClassPath=/BRCosmetics/Athena/UI/Frontend/Spatial/BPS12_AboutModal
   bIsVariable=true
   parent=Switcher_Content
   widget=BPS12_AboutModal
   properties {
    header=Style Challenges
    header_Icon=/Game/UI/Foundation/Textures/Icons/Locker/MI_Icon_Outfit_Blue_128
    subHeader=Available for Battle Pass Owners
    title="New outfits && styles!"
    divider_Color=<rgba 0.25 0.875 1 1>
    entries [
     "<text color=\"FFFFFFFF\" fontface=\"black\" case=\"upper\">Level up</> to unlock new <text color=\"FFFFFFFF\" fontface=\"black\" case=\"upper\">Battle Pass Outfits</>."
     "<text color=\"FFFFFFFF\" fontface=\"black\" case=\"upper\">Complete challenges</> to unlock new <text color=\"FFFFFFFF\" fontface=\"black\" case=\"upper\">Styles</>."
     "Complete <text color=\"FFFFFFFF\" fontface=\"black\" case=\"upper\">aquaman's challenges</> to unlock his <text color=\"FFFFFFFF\" fontface=\"black\" case=\"upper\">Outfit </>and <text color=\"FFFFFFFF\" fontface=\"black\" case=\"upper\">bonus style</>."
    ]
    subheader_Color=<rgba 0 0.6667 1 1>
   }
   slot {class=WidgetSwitcherSlot}
  }
 ]
}
